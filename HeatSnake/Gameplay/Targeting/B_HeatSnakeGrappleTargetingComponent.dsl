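; Targeting component for the HeatSnake grapple weapon that relays hookshot targeting state to listeners and drives the grapple reticle.
; On cosmetic-capable instances, an async tag-count listener on the cached player pawn broadcasts hookshot targeting changes, gated on no movement cooldown and a valid-target tag.
; Native reticle updates forward to either HeatSnake weapon class while off cooldown.
; Target type, hit info and server validation are re-broadcast through delegates.
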(event EventBeginPlay
  (bind _playerpawn (Default|GetPlayerPawn))
  ;[0.08048 0.5521 0.14] BeginPlay: Bind to know when the player is targeting the hookshot ability and fire off delegates
  (|CallParentFunction)
  (Networking|CanExecuteCosmeticEvents
    (:True
      (Utilities|IsValid _playerpawn
        (:Is%20Valid
          (bind (_asynctaskproxy _tagcount) (Ability|Tasks|WaitGameplayTagCountChangedOnActor _playerpawn "(TagName=\"ChainWhipGameplay.Abilities.Hookshot\")")
            (:then
              (Variables|Grapple|SetAsyncAction_WaitGameplayTagCountChanged _asynctaskproxy))
            (:TagCountChanged
              (if (and (not (GameplayTags|HasMatchingGameplayTag (Ability|GetAbilitySystemComponentfromActor (Components|GetOwner)) "(TagName=\"Granted.Cooldown.HeatSnake.Movement\")")) (GameplayTags|HasMatchingGameplayTag (Ability|GetAbilitySystemComponentfromActor (Components|GetOwner)) "(TagName=\"Granted.HeatSnake.Targeting.HaveValidTarget\")"))
                (Variables|Grapple|SetbTargetingHookshot (> _tagcount 0))
                (Default|CallOnTargetingHookshot (> _tagcount 0))))))))))

(event EventEndPlay
  (params (EndPlayReason :type EEndPlayReason%20Enum :default "Destroyed"))
  ;[0.1146 0.05059 0.0144] EndPlay: Clean up actions.
  (Utilities|FlowControl|Sequence
    (:then_0
      (Networking|CanExecuteCosmeticEvents
        (:True
          (bind _asyncaction_waitgameplaytagcountchanged (Variables|Grapple|GetAsyncAction_WaitGameplayTagCountChanged)
            (:then
              (Ability|Async|EndAction _asyncaction_waitgameplaytagcountchanged))
            (:else)))))
    (:then_1
      (|CallParentFunction EndPlayReason))))

(event EventUpdateReticle
  (params (bHitSomething :type Boolean) (bUseBlocked :type Boolean) (bIsWeaponAimingAtTarget :type Boolean) (TargetPos :type Vector%20%28by%20ref%29 :default "0, 0, 0"))
  (bind _returnvalue (Components|GetOwner))
  (if (not (GameplayTags|HasMatchingGameplayTag (Ability|GetAbilitySystemComponentfromActor _returnvalue) "(TagName=\"Granted.Cooldown.HeatSnake.Movement\")"))
    (bind _asb_heat_snake_weapon (Utilities|Casting|CastToB_HeatSnake_Weapon _returnvalue)
      (:then
        (Class|BHeatSnakeWeapon|UpdateGrappleReticlefromNative _asb_heat_snake_weapon bHitSomething bUseBlocked bIsWeaponAimingAtTarget TargetPos))
      (:CastFailed
        (bind _asb_heat_snake_weapon_grapple (Utilities|Casting|CastToB_HeatSnake_Weapon_Grapple _returnvalue))
        (Class|BHeatSnakeWeaponGrapple|UpdateGrappleReticlefromNative _asb_heat_snake_weapon_grapple bHitSomething bUseBlocked bIsWeaponAimingAtTarget TargetPos)))
    (|CallParentFunction bHitSomething bUseBlocked bIsWeaponAimingAtTarget TargetPos)))

(event EventOnServerValidatedTarget
  (Default|CallOnServerTargetValidated))

(fn GetPlayerPawn
  ;[0.1094 0.03257 0.1065] Return the player pawn that is holding the owner weapon actor.
  (Utilities|FlowControl|Sequence
    (:then_0
      (Variables|Default|GetCachedPlayerPawn
        (:then)
        (:else
          (bind _asfort_weapon (Utilities|Casting|CastToFortWeapon (Components|GetOwner)))
          (bind _asfort_player_pawn_athena (Utilities|Casting|CastToFortPlayerPawnAthena (Actor|GetOwner _asfort_weapon)))
          (Variables|Default|SetCachedPlayerPawn _asfort_player_pawn_athena))))
    (:then_1
      (return (Variables|Default|GetCachedPlayerPawn)))))

(fn SetTargetType
  (params (NewTargetType :type HeatSnakeGrappleTargetTypeEnum%20Enum :default "NewEnumerator0"))
  ;[0.04688 0.01078 0.04685] Sets the target type under the reticle.
  ; Calculated and called from the gameplay ability - used by HeatSnakeWeapon
  (Networking|CanExecuteCosmeticEvents
    (:True
      (if (Utilities|Enum|NotEqual(Enum) NewTargetType (Variables|Reticle|Targeting|GetTargetType))
        (bind _output_get (Variables|Reticle|Targeting|SetTargetType NewTargetType))
        (Default|CallOnTargetingTargetUpdated _output_get)))))

(fn UpdateTargetInformation
  (params (HitResult :type Hit%20Result%20Structure) (TargetType :type HeatSnakeGrappleTargetTypeEnum%20Enum :default "NewEnumerator0"))
  (Default|CallOnTargetInfoUpdated HitResult TargetType))
