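; Targeting component for the HeatSnake weapon that supplies the primary attack's target filter to gameplay abilities and tracks what is under the reticle.
; ChainWhip hookshot and melee targeting abilities push the target type, which updates only on change where cosmetic events can execute, then fires a dispatcher.
; A separate update relays hit result and target type through another dispatcher.
; Helpers resolve the player pawn holding the owning weapon, with caching.
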(fn GetAttackTargetFilter
  ;[0.02909 0.1458 0.0831] Returns the target filter for the primary attack. Used by GAs.
  (return (Variables|Attack|GetAttackTargetFilter)))

(fn GetPlayerPawn
  ;[0.1094 0.03257 0.1065] Return the player pawn that is holding the owner weapon actor.
  (Utilities|FlowControl|Sequence
    (:then_0
      (Variables|Transient|GetCachedPlayerPawn
        (:then)
        (:else
          (bind _asfort_weapon (Utilities|Casting|CastToFortWeapon (Components|GetOwner)))
          (bind _asfort_player_pawn_athena (Utilities|Casting|CastToFortPlayerPawnAthena (Actor|GetOwner _asfort_weapon)))
          (Variables|Transient|SetCachedPlayerPawn _asfort_player_pawn_athena))))
    (:then_1
      (return (Variables|Transient|GetCachedPlayerPawn)))))

(fn GetPlayerPawnConst
  ;[0.1094 0.03257 0.1065] Return the player pawn that is holding the owner weapon actor.
  (Utilities|FlowControl|Sequence
    (:then_0
      (bind _cachedplayerpawn (Variables|Transient|GetCachedPlayerPawn)
        (:then
          (return _cachedplayerpawn))
        (:else
          (bind _asfort_weapon (Utilities|Casting|CastToFortWeapon (Components|GetOwner)))
          (bind _asfort_player_pawn_athena (Utilities|Casting|CastToFortPlayerPawnAthena (Actor|GetOwner _asfort_weapon)))
          (return _asfort_player_pawn_athena))))
    (:then_1
      (return 0))))

(fn SetTargetType
  (params (TargetType :type ChainWhipTargetActorTypeEnum%20Enum :default "NewEnumerator0"))
  ;[0.04688 0.01078 0.04685] Sets the target type under the reticle.
  ; Calculated and called from the gameplay ability - GA_ChainWhip_Secondary_Hookshot and GA_ChainWhip_Melee_Targeting
  (Networking|CanExecuteCosmeticEvents
    (:True
      (if (Utilities|Enum|NotEqual(Enum) TargetType (Variables|Reticle|Targeting|Transient|GetTargetType))
        (bind _output_get (Variables|Reticle|Targeting|Transient|SetTargetType TargetType))
        (Default|CallOnTargetingTargetUpdated _output_get)))))

(fn UpdateTargetInformation
  (params (HitResult :type Hit%20Result%20Structure) (TargetType :type ChainWhipTargetActorTypeEnum%20Enum :default "NewEnumerator0"))
  (Default|CallOnTargetInfoUpdated HitResult TargetType))
